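# Masked, lit material for the Temple Female glider cosmetic, adding animated cloth wrinkles and flutter to MF_GliderBaseShader.
# Three panning fold normals, blended by secondary noise, are added to the base normal.
# Swirl noise drives world position offset for flutter and sine-animated torn cloth, shaped by channels of a wrinkles mask texture.
# Pre-skinned positions supply projection coordinates; several alternative wrinkle branches, including flow-map and world-aligned versions, appear unconnected.
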Material /BRCosmetics/Gadgets/Assets/VinderTech_GliderChute/Glider_TempleFemale/Materials/M_Glider_Temple_Female
graph {
 properties {
  bUseMaterialAttributes=true
  bUsedWithInstancedStaticMeshes=true
  bUsedWithMeshDeformer=true
  bUsedWithSkeletalMesh=true
  blendMode=BLEND_Masked
  pixelDepthOffsetMode=PDOM_Legacy
  usedShadingModels=Default Lit
 }
 nodes [
  Constant2Vector() {r=4}
  n1 = Constant2Vector() {g=0.51, r=0.321}
  n2 = ScalarParameter() {defaultValue=1, group=Wrinkles, parameterName=TurbulenceSpeed}
  n3 = Divide(A=VertexInterpolator(VS=NamedRerouteUsage()).PS, B=ScalarParameter() {defaultValue=50, group=Wrinkles, parameterName=WrinklesSize})
  n4 = TextureSample(UVs=Panner(Coordinate=@n3, Speed=Multiply(A=@n1, B=@n2)) {speedX=0.321, speedY=0.51}) {
   samplerSource=SSM_Wrap_WorldGroupSettings
   samplerType=SAMPLERTYPE_Normal
   texture=/Game/Gadgets/Assets/VinderTech_GliderChute/Glider_BlackMondayCape/Textures/T_Folds_N
  }
  n5 = Constant2Vector() {g=-0.527, r=0.23}
  n6 = TextureSample(UVs=Panner(Coordinate=Multiply(A=@n3) {constB=-1}, Speed=Multiply(A=@n5, B=@n2)) {speedX=0.23, speedY=-0.21}) {
   samplerSource=SSM_Wrap_WorldGroupSettings
   samplerType=SAMPLERTYPE_Normal
   texture=&T_Folds_N
  }
  n7 = Constant2Vector() {g=-0.2721, r=0.021}
  n8 = TextureSample(UVs=Panner(Coordinate=@n3, Speed=Multiply(A=@n7, B=@n2)) {speedX=0.021, speedY=-0.2721}) {
   samplerSource=SSM_Wrap_WorldGroupSettings
   samplerType=SAMPLERTYPE_Normal
   texture=&T_Folds_N
  }
  Divide(A=Add(A=Add(A=@n4.RGB, B=@n6.RGB), B=@n8.RGB), B=Constant3Vector() {constant=<rgba 1 1 3 0>})
  n9 = VertexInterpolator(VS=NamedRerouteUsage())
  n10 = Divide(A=Add(A=@n9.PS)) {constB=30}
  n11 = Divide(A=Add(A=@n9.PS)) {constB=-60}
  n12 = Add(
   A=TextureSample(UVs=Panner(Coordinate=@n10) {speedX=0.1, speedY=0.3}) {samplerType=SAMPLERTYPE_Normal, texture=&T_Folds_N}.RGB
   B=TextureSample(UVs=Panner(Coordinate=@n11) {speedX=-0.1, speedY=0.21}) {samplerType=SAMPLERTYPE_Normal, texture=&T_Folds_N}.RGB
  )
  n13 = Transform(None=VertexNormalWS()) {transformSourceType=TRANSFORMSOURCE_World, transformType=TRANSFORM_Local}
  n14 = ComponentMask(None=@n13) {g=true, r=true}
  n15 = Panner(Coordinate=MaterialFunctionCall!(
    1=Multiply(A=@n14) {constB=-1}
    2=ComponentMask(None=@n13) {b=true, r=true}
    3=ComponentMask(None=@n13) {b=true, g=true}
    4
    A=NamedRerouteUsage()
   ) {materialFunction=/Engine/Functions/Engine_MaterialFunctions02/Lerp_Multiple_Float2}.Lerp 3 Inputs) {speedX=-0.1, speedY=0.23}
  Divide(
   A=Add(
    A=@n12
    B=TextureSample(UVs=@n15) {samplerType=SAMPLERTYPE_Normal, texture=&T_Folds_N}.RGB
   )
   B=Constant3Vector() {constant=<rgba 1 1 3 0>}
  )
  Divide(A=@n12, B=Constant3Vector() {constant=<rgba 1 1 2 0>})
  n16 = Transform(None=VertexNormalWS()) {transformSourceType=TRANSFORMSOURCE_World, transformType=TRANSFORM_Local}
  n17 = ComponentMask(None=@n16) {g=true, r=true}
  n18 = ComponentMask(None=MaterialFunctionCall(
    1=Multiply(A=@n17) {constB=-1}
    2=ComponentMask(None=@n16) {b=true, r=true}
    3=ComponentMask(None=@n16) {b=true, g=true}
    A=NamedRerouteUsage()
   ) {materialFunction=&Lerp_Multiple_Float2}.Lerp 3 Inputs) {g=true, r=true}
  n19 = Time()
  n20 = TextureCoordinate() {uTiling=2, vTiling=2}
  n21 = TextureSample(UVs=Panner(Coordinate=@n20) {speedX=0.3, speedY=0.5}) {
   samplerSource=SSM_Wrap_WorldGroupSettings
   texture=/Game/Gadgets/Assets/VinderTech_GliderChute/Glider_Angler/FX/T_TilingNoise06_HorizontalAlphaGrad
  }
  n22 = TextureSample(UVs=Panner(Coordinate=@n20) {speedX=-0.23, speedY=0.21}) {
   samplerSource=SSM_Wrap_WorldGroupSettings
   texture=&T_TilingNoise06_HorizontalAlphaGrad
  }
  n23 = TextureSample(UVs=Panner(Coordinate=@n20) {speedY=-0.3}) {
   samplerSource=SSM_Wrap_WorldGroupSettings
   texture=&T_TilingNoise06_HorizontalAlphaGrad
  }
  LinearInterpolate(
   A=Constant3Vector() {constant=<rgba 0 0 1 0>}
   B=TextureSample(UVs=Add(A=@n18, B=@n19)) {samplerType=SAMPLERTYPE_Normal, texture=&T_Folds_N}.RGB
   Alpha=Clamp(None=LinearInterpolate(A=@n21.R, B=@n22.R, Alpha=@n23.R))
  )
  n24 = VertexNormalWS()
  n25 = ComponentMask(None=@n24) {g=true, r=true}
  n26 = ComponentMask(None=@n24) {b=true, r=true}
  n27 = ComponentMask(None=@n24) {b=true, g=true}
  n28 = NamedRerouteUsage()
  n29 = Time()
  MaterialFunctionCall!(
   Diffuse
   Normal Map=TextureObject() {samplerType=SAMPLERTYPE_Normal, texture=&T_Folds_N}
   "Flow Vector Map (see tooltip)"=LinearInterpolate(
    A=NamedRerouteUsage()
    B=MaterialFunctionCall(1=@n25, 2=@n26, 3=@n27, A=@n28) {materialFunction=&Lerp_Multiple_Float2}.Lerp 3 Inputs
   )
   UV texture 2 offset and Scale
   "Time (see tooltip)"=Add(A=Multiply(A=@n29) {constB=-2}, B=TextureSample() {texture=&AtmosphericCloudNoise01}.R)
   UVs=TextureCoordinate() {uTiling=10, vTiling=10}
  ) {materialFunction=/FortWater/Materials/Functions/WaterFlowMap}
  n30 = GetMaterialAttributes(None=MaterialFunctionCall!(UV 1 coordinates) {materialFunction=/Game/Gadgets/Assets/VinderTech_GliderChute/Materials/MF_GliderBaseShader}.Result Material Attributes) {attributeGetTypes [#1, #2, #3]}
  n31 = Constant2Vector() {g=0.51, r=0.321}
  n32 = ScalarParameter() {defaultValue=1, group=Wrinkles, parameterName=SecondaryTurbulenceSpeed, sortPriority=300}
  n33 = Divide(
   A=VertexInterpolator(VS=NamedRerouteUsage()).PS
   B=ScalarParameter() {defaultValue=200, group=Wrinkles, parameterName=WrinklesSecondarySize, sortPriority=280}
  )
  n34 = TextureSample(UVs=Panner(Coordinate=@n33, Speed=Multiply(A=@n31, B=@n32)) {speedX=0.3, speedY=0.5}) {
   samplerSource=SSM_Wrap_WorldGroupSettings
   texture=&T_TilingNoise06_HorizontalAlphaGrad
  }
  n35 = Constant2Vector() {g=-0.21, r=0.23}
  n36 = TextureSample(UVs=Panner(Coordinate=@n33, Speed=Multiply(A=@n35, B=@n32)) {speedX=-0.23, speedY=0.21}) {
   samplerSource=SSM_Wrap_WorldGroupSettings
   texture=&T_TilingNoise06_HorizontalAlphaGrad
  }
  n37 = Constant2Vector() {g=-0.2721, r=0.021}
  n38 = TextureSample(UVs=Panner(Coordinate=@n33, Speed=Multiply(A=@n37, B=@n32)) {speedY=-0.3}) {
   samplerSource=SSM_Wrap_WorldGroupSettings
   texture=&T_TilingNoise06_HorizontalAlphaGrad
  }
  n39 = LinearInterpolate(A=@n34.R, B=@n36.R, Alpha=@n38.R)
  n40 = Normalize(VectorInput=MaterialFunctionCall(1=@n4.RGB, 2=@n8.RGB, 3=@n6.RGB, A=Clamp(None=@n39)) {materialFunction=&Lerp_Multiple_Float3}.Lerp 3 Inputs)
  n41 = NamedRerouteUsage()
  n42 = Add(
   A=@n30.Normal
   B=Multiply(A=ScalarParameter() {defaultValue=0.3, group=Wrinkles, parameterName=WrinklesNormalStrength}, B=Multiply(A=@n40, B=@n41))
  ) {constB=0}
  n43 = Power(Base=NamedRerouteUsage(), Exp=ScalarParameter() {defaultValue=2.5, group=WPO, parameterName=FlutterFalloff})
  n44 = NamedRerouteUsage()
  n45 = LinearInterpolate(
   A=ScalarParameter() {defaultValue=150, group=WPO, parameterName=WPO_TextureSize}
   B=ScalarParameter() {defaultValue=30, group=WPO, parameterName=TornCloth_TextureSize}
   Alpha=NamedRerouteUsage()
  ) {constB=50}
  n46 = MaterialFunctionCall(
   X=ScalarParameter() {defaultValue=0.27, group=WPO, parameterName=WPO_SpeedX}
   Y=ScalarParameter() {defaultValue=2, group=WPO, parameterName=WPO_SpeedY}
  ) {materialFunction=&MakeFloat2}
  n47 = Multiply(A=@n46.Result, B=Constant2Vector() {g=-1, r=1})
  n48 = NamedRerouteUsage()
  n49 = TextureSample(UVs=Panner(Coordinate=Divide(A=@n44, B=@n45), Speed=LinearInterpolate(A=@n47, B=@n46.Result, Alpha=@n48)) {speedY=2}) {samplerType=SAMPLERTYPE_Normal, texture=/Game/Global/Textures/Default/Noise/T_SwirlNoise_N}
  n50 = Normalize(VectorInput=@n49.RGB)
  n51 = NamedRerouteUsage()
  n52 = Time()
  n53 = ScalarParameter() {defaultValue=6, group=Torn Cloth, parameterName=TornClothSpeed}
  n54 = Saturate(None=ComponentMask(None=PreSkinnedPosition()) {r=true})
  n55 = Add(A=@n49.RGB, B=Multiply(A=Multiply(A=@n52, B=@n53), B=LinearInterpolate(Alpha=@n54) {constA=0.9}))
  n56 = Add(A=@n49.RGB, B=Add(A=Sine(None=@n55)))
  n57 = ScalarParameter() {defaultValue=1.5, group=Torn Cloth, parameterName=TornClothAmplitude}
  n58 = Add(
   A=Multiply(A=Multiply(A=@n43, B=@n50), B=ScalarParameter() {defaultValue=2, group=WPO, parameterName=FlutterAmplitude}) {constB=2}
   B=Multiply(
    A=Multiply(A=@n51, B=@n56)
    B=MaterialFunctionCall(Z=@n57) {materialFunction=&MakeFloat3}.Result
   )
  )
  MaterialFunctionCall(In=SetMaterialAttributes!(
    MaterialAttributes=@n30.MaterialAttributes
    Opacity Mask=@n30.MaterialAttributes
    Normal=Normalize(VectorInput=@n42)
    World Position Offset=Add(A=@n30.World Position Offset, B=@n58) {constB=0}
   ) {attributeSetTypes [#1, #2, #3]}) {materialFunction=&MF_TechArtOverrides_accessories}
  n59 = PreSkinnedPosition()
  n60 = NamedRerouteUsage()
  n61 = Multiply(A=Constant2Vector() {g=-1, r=1}, B=@n60)
  Multiply(
   A=NamedRerouteUsage()
   B=MaterialFunctionCall!(
    TextureObject=TextureObject() {samplerType=SAMPLERTYPE_Normal, texture=&T_Folds_N}
    TextureSize=VectorParameter() {parameterName=WrinklesTiling, defaultValue=<rgba 30 30 30 1>}.RGB
    Normal
    WorldSpace
    ProjectionTransitionContrast
    WorldPosition=VertexInterpolator(VS=@n59).PS
    Use High Quality Normals
    Pan_YZ=@n61
    Pan_XZ=@n61
    Pan_XY=@n60
   ) {materialFunction=/Game/Weapons/WeaponSkins/Wraps/MaterialFunctions/WorldAlignedNormal_ComplexPan}.XYZ Texture
  )
  n62 = TextureSample() {
   samplerType=SAMPLERTYPE_Masks
   texture=/BRCosmetics/Gadgets/Assets/VinderTech_GliderChute/Glider_TempleFemale/Textures/T_Glider_Temple_Female_WrinklesMasks
  }
  NamedRerouteDeclaration(None=@n62.A) {name=ClothDirection, nodeColor=<rgba 1 0 0.1653 1>}
  NamedRerouteDeclaration(None=MaterialFunctionCall(X=ScalarParameter() {defaultValue=1, parameterName=WrinklesSpeedX}, Y=ScalarParameter() {defaultValue=2, parameterName=WrinklesSpeedY}) {materialFunction=&MakeFloat2}.Result) {name=WrinklesSpeed, nodeColor=<rgba 1 0 0.8907 1>}
  n63 = NamedRerouteUsage()
  n64 = NamedRerouteUsage()
  NamedRerouteDeclaration(None=Power(Base=Multiply(A=@n63, B=@n64), Exp=ScalarParameter() {defaultValue=1.5, group=Torn Cloth, parameterName=TornClothFalloff})) {name=TornClothMask, nodeColor=<rgba 0 1 0.4671 1>}
  NamedRerouteDeclaration(None=Ceil(None=@n63)) {name=TornColthCeiledMask, nodeColor=<rgba 0 0.6953 1 1>}
  NamedRerouteDeclaration(None=@n62.B) {name=MaskBlue, nodeColor=<rgba 0 1 0.8359 1>}
  n65 = PreSkinnedPosition()
  NamedRerouteDeclaration(None=MaterialFunctionCall(
    1=ComponentMask(None=@n65) {g=true, r=true}
    2=ComponentMask(None=@n65) {b=true, r=true}
    3=ComponentMask(None=@n65) {b=true, g=true}
    A=NamedRerouteUsage()
   ) {materialFunction=&Lerp_Multiple_Float2}.Lerp 3 Inputs) {name=SkinnedCoordinates, nodeColor=<rgba 1 0 0.7734 1>}
  NamedRerouteDeclaration(None=@n62.G) {name=MaskGreen, nodeColor=<rgba 0 0.3281 0.1015 1>}
  NamedRerouteDeclaration(None=@n62.R) {name=MaskRed, nodeColor=<rgba 1 0 0.0557 1>}
  TextureCoordinate() {uTiling=2, vTiling=2}
  Time()
 ]
}
